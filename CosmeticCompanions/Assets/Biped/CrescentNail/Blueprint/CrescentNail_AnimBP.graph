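# Animation Blueprint for CrescentNail, a biped cosmetic companion, covering locomotion, sitting, emotes, reactions and look-at.
# The high-fidelity path layers start, mood, body-pose, turn and head-antic additives over a locomotion state machine, then applies an aim offset.
# A simpler low-fidelity path is cached alongside it, and bActiveValue selects between them with inertialized blending.
# The result passes through dead blending, slope warping and leg IK, while Chooser proxy tables supply stops, idles and poses.
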AnimBlueprint /CosmeticCompanions/Assets/Biped/CrescentNail/Blueprint/CrescentNail_AnimBP
target_skeleton=/CosmeticCompanions/Assets/Biped/CrescentNail/Meshes/CrescentNail_Biped_Skeleton
graphs [
 AllowsCinematicOverride {nodes [
   AllowsCinematicOverride() {is=title K2Node_FunctionEntry}
   > Return Node!(execute, ReturnValue:bool=false)
  ]}
 AnimGraph {nodes [
   n1 = Slot 'SequencerDefault'(Source=Blend Poses by bool(
     BlendPose_0=Use cached pose 'LowFidelityPose'()
     BlendPose_1=Use cached pose 'HighFidelityPose'()
     BlendTime_0=0.2
     BlendTime_1=0.2
     bActiveValue=Get bActiveValue() -> (bActiveValue:bool)
    ) {blendType=Linear, transitionType=Inertialization}) {slotName=SequencerDefault}
   n2 = Biped Slope Warping!(
    ComponentPose:Component Space Pose Link Structure=Local To Component(LocalPose=Dead Blending!(Source:Pose Link Structure=@n1) -> (Pose:Pose Link Structure) {is=title AnimGraphNode_DeadBlending, defaultBlendMode=Linear})
    Alpha:float=1
   ) -> (Pose:Component Space Pose Link Structure) {
    is=title FortAnimGraphNode_SlopeWarping
    alphaBoolBlend {blendOption=Linear}
    feetDefinitions [
     {numBonesInLimb=3, fKFootBone {boneName=ball_l}, iKFootBone {boneName=IK_Foot_Rear_L}}
     {numBonesInLimb=3, fKFootBone {boneName=ball_r}, iKFootBone {boneName=IK_Foot_Rear_R}}
    ]
    iKFootRootBone {boneName=IK_Foot_Front_Root}
    pelvisBone {boneName=root}
   }
   n3 = Leg IK!(ComponentPose:Component Space Pose Link Structure=@n2, Alpha:float=1) -> (Pose:Component Space Pose Link Structure) {
    is=title AnimGraphNode_LegIK
    alphaBoolBlend {blendOption=Linear}
    legsDefinition [
     {
      bEnableKneeTwistCorrection=true
      bEnableRotationLimit=false
      footBoneForwardAxis=Y
      hingeRotationAxis=None
      minRotationAngle=15
      numBonesInLimb=3
      twistOffsetCurveName=None
      fKFootBone {boneName=ball_l}
      iKFootBone {boneName=IK_Foot_Rear_L}
     }
     {
      bEnableKneeTwistCorrection=true
      bEnableRotationLimit=false
      footBoneForwardAxis=Y
      hingeRotationAxis=None
      minRotationAngle=15
      numBonesInLimb=3
      twistOffsetCurveName=None
      fKFootBone {boneName=ball_r}
      iKFootBone {boneName=IK_Foot_Rear_R}
     }
    ]
   }
   Output Pose(Result=Component To Local(ComponentPose=@n3))
   AnimGraphNode_StateMachine_6 = Head Antic() {is=title AnimGraphNode_StateMachine, bAllowConduitEntryStates=true}
   Save cached pose 'Head Antic Pose'(Pose=@AnimGraphNode_StateMachine_6) {cacheName=Head Antic Pose}
   AnimGraphNode_StateMachine_0 = LocomotionStateMachine() {is=title AnimGraphNode_StateMachine, bAllowConduitEntryStates=true}
   AnimGraphNode_StateMachine_5 = LocomotionStartAdditive() {is=title AnimGraphNode_StateMachine}
   AnimGraphNode_StateMachine_3 = BodyMoodAdditive() {is=title AnimGraphNode_StateMachine, bAllowConduitEntryStates=true}
   n4 = Apply Additive(
    Base=Apply Additive(Base=@AnimGraphNode_StateMachine_0, Additive=@AnimGraphNode_StateMachine_5) {alphaBoolBlend {blendOption=Linear}}
    Additive=@AnimGraphNode_StateMachine_3
   ) {alphaBoolBlend {blendOption=Linear}}
   AnimGraphNode_StateMachine_1 = Body Pose Additive() {is=title AnimGraphNode_StateMachine, bAllowConduitEntryStates=true}
   n5 = Get AnticAlpha() -> (AnticAlpha:double)
   n6 = Get CompanionDirection() -> (CompanionDirection:float)
   n7 = Apply Additive(
    Base=Apply Additive(
     Base=Apply Additive(Base=@n4, Additive=@AnimGraphNode_StateMachine_1, Alpha=@n5) {alphaBoolBlend {blendOption=Linear}}
     Additive=Blendspace Player 'CrescentNail_COR_Loco_TurnTail_BS'!(X=@n6) -> (Pose) {blendSpace=/CosmeticCompanions/Assets/Biped/CrescentNail/Animation/Core/CrescentNail_COR_Loco_TurnTail_BS}
    ) {alphaBoolBlend {blendOption=Linear}}
    Additive=Blendspace Player 'CrescentNail_COR_Loco_TurnBody_BS'!(X=Get CompanionDirection()) -> (Pose) {blendSpace=/CosmeticCompanions/Assets/Biped/CrescentNail/Animation/Core/CrescentNail_COR_Loco_TurnBody_BS}
    Alpha=Get BodyBendAlpha() -> (BodyBendAlpha:double)
   ) {alphaBoolBlend {blendOption=Linear}}
   n8 = Use cached pose 'Head Antic Pose'()
   n9 = Get AnticAlpha()
   n10 = AimOffset Player 'CrescentNail_AO'!(
    X:float=Get LookAtYawAdjusted() -> (LookAtYawAdjusted:double)
    Y:float=Get LookAtPitchAdjusted() -> (LookAtPitchAdjusted:double)
    BasePose:Pose Link Structure=Slot 'DefaultSlot'(Source=Apply Additive(Base=@n7, Additive=@n8, Alpha=@n9) {alphaBoolBlend {blendOption=Linear}})
    AlphaCurveName:Name=DisableAimOffset
   ) -> (Pose:Pose Link Structure) {
    alphaInputType=Curve
    blendSpace=/CosmeticCompanions/Assets/Biped/CrescentNail/Animation/AimOffsets/CrescentNail_AO
    alphaBoolBlend {blendOption=Linear}
    alphaScaleBiasClamp {bias=1, scale=-1}
   }
   Save cached pose 'HighFidelityPose'(Pose=@n10) {cacheName=HighFidelityPose}
   AnimGraphNode_StateMachine_4 = LowFidelityStateMachine() {is=title AnimGraphNode_StateMachine, bAllowConduitEntryStates=true}
   n11 = Apply Additive(Base=@AnimGraphNode_StateMachine_4, Additive=Use cached pose 'Head Antic Pose'(), Alpha=Get AnticAlpha()) {alphaBoolBlend {blendOption=Linear}}
   n12 = Blendspace Player 'CrescentNail_COR_Loco_TurnBody_BS'(X=Get CompanionDirection()) {blendSpace=&CrescentNail_COR_Loco_TurnBody_BS}
   n13 = AimOffset Player 'CrescentNail_AO'!(
    X=Get LookAtYawAdjusted()
    Y=Get LookAtPitch() -> (LookAtPitch:float)
    BasePose=Slot 'DefaultSlot'(Source=Apply Additive(Base=@n11, Additive=@n12) {alphaBoolBlend {blendOption=Linear}})
    Alpha:float=1
   ) -> (Pose) {blendSpace=&CrescentNail_AO, alphaBoolBlend {blendOption=Linear}}
   Save cached pose 'LowFidelityPose'(Pose=@n13) {cacheName=LowFidelityPose}
  ]}
 LocomotionStateMachine {
  entry=Conduit
  path=AnimGraph.AnimGraphNode_StateMachine_0.LocomotionStateMachine
  ids=id
  nodes [
   AnimStateConduitNode_0 = Conduit()
   AnimStateNode_0 = EmoteLoops() {is=title AnimStateNode}
   AnimStateNode_1 = TransitionToDefaultIdle() {is=title AnimStateNode}
   AnimStateNode_2 = LocoStop() {is=title AnimStateNode}
   AnimStateNode_4 = Locomotion() {is=title AnimStateNode}
   AnimStateNode_5 = Idle() {bAlwaysResetOnEntry=true}
   AnimStateTransitionNode_0 = AnimStateTransitionNode() {bAutomaticRuleBasedOnSequencePlayerInState=true}
   AnimStateTransitionNode_1 = AnimStateTransitionNode() {crossfadeDuration=0.35, priorityOrder=100}
   AnimStateTransitionNode_10 = AnimStateTransitionNode() {logicType=TLT_Inertialization}
   AnimStateTransitionNode_11 = AnimStateTransitionNode() {logicType=TLT_Inertialization}
   AnimStateTransitionNode_12 = AnimStateTransitionNode() {logicType=TLT_Inertialization, priorityOrder=2}
   AnimStateTransitionNode_13 = AnimStateTransitionNode()
   AnimStateTransitionNode_14 = AnimStateTransitionNode()
   AnimStateTransitionNode_15 = AnimStateTransitionNode() {bAutomaticRuleBasedOnSequencePlayerInState=true}
   AnimStateTransitionNode_16 = AnimStateTransitionNode() {
    crossfadeDuration=0.3
    logicType=TLT_Inertialization
    blendProfileWrapper {blendProfile=&CrescentNail_Biped_Skeleton:FastFeetSlowHead}
   }
   AnimStateTransitionNode_17 = AnimStateTransitionNode() {logicType=TLT_Inertialization}
   AnimStateTransitionNode_2 = AnimStateTransitionNode()
   AnimStateTransitionNode_4 = AnimStateTransitionNode() {
    logicType=TLT_Inertialization
    blendProfileWrapper {blendProfile=&CrescentNail_Biped_Skeleton:FastFeetSlowHead}
   }
  ]
  links [
   {
    dst=EmoteLoops
    rule=AnimGraph.AnimGraphNode_StateMachine_0.LocomotionStateMachine.AnimStateTransitionNode_10.AnimationTransitionGraph_0
    src=Conduit
    via=AnimStateTransitionNode_10
   }
   {
    dst=Idle
    rule=AnimGraph.AnimGraphNode_StateMachine_0.LocomotionStateMachine.AnimStateTransitionNode_1.AnimationTransitionGraph_0
    src=Conduit
    via=AnimStateTransitionNode_1
   }
   {
    dst=Locomotion
    rule=AnimGraph.AnimGraphNode_StateMachine_0.LocomotionStateMachine.AnimStateTransitionNode_16.AnimationTransitionGraph_0
    src=Conduit
    via=AnimStateTransitionNode_16
   }
   {
    dst=TransitionToDefaultIdle
    rule=AnimGraph.AnimGraphNode_StateMachine_0.LocomotionStateMachine.AnimStateTransitionNode_12.AnimationTransitionGraph_0
    src=Conduit
    via=AnimStateTransitionNode_12
   }
   {
    dst=Conduit
    rule=AnimGraph.AnimGraphNode_StateMachine_0.LocomotionStateMachine.AnimStateTransitionNode_14.AnimationTransitionGraph_0
    src=EmoteLoops
    via=AnimStateTransitionNode_14
   }
   {
    dst=TransitionToDefaultIdle
    rule=AnimGraph.AnimGraphNode_StateMachine_0.LocomotionStateMachine.AnimStateTransitionNode_11.AnimationTransitionGraph_0
    src=EmoteLoops
    via=AnimStateTransitionNode_11
   }
   {
    dst=Conduit
    rule=AnimGraph.AnimGraphNode_StateMachine_0.LocomotionStateMachine.AnimStateTransitionNode_2.AnimationTransitionGraph_0
    src=Idle
    via=AnimStateTransitionNode_2
   }
   {
    dst=Conduit
    rule=AnimGraph.AnimGraphNode_StateMachine_0.LocomotionStateMachine.AnimStateTransitionNode_0.AnimationTransitionGraph_0
    src=LocoStop
    via=AnimStateTransitionNode_0
   }
   {
    dst=Locomotion
    rule=AnimGraph.AnimGraphNode_StateMachine_0.LocomotionStateMachine.AnimStateTransitionNode_17.AnimationTransitionGraph_0
    src=LocoStop
    via=AnimStateTransitionNode_17
   }
   {
    dst=LocoStop
    rule=AnimGraph.AnimGraphNode_StateMachine_0.LocomotionStateMachine.AnimStateTransitionNode_4.AnimationTransitionGraph_0
    src=Locomotion
    via=AnimStateTransitionNode_4
   }
   {
    dst=Conduit
    rule=AnimGraph.AnimGraphNode_StateMachine_0.LocomotionStateMachine.AnimStateTransitionNode_13.AnimationTransitionGraph_0
    src=TransitionToDefaultIdle
    via=AnimStateTransitionNode_13
   }
   {
    dst=Conduit
    rule=AnimGraph.AnimGraphNode_StateMachine_0.LocomotionStateMachine.AnimStateTransitionNode_15.AnimationTransitionGraph_0
    src=TransitionToDefaultIdle
    via=AnimStateTransitionNode_15
   }
  ]
 }
 Conduit {path=AnimGraph.AnimGraphNode_StateMachine_0.LocomotionStateMachine.AnimStateConduitNode_0.Conduit, nodes [Result(bCanEnterTransition=true)]}
 EmoteLoops {
  path=AnimGraph.AnimGraphNode_StateMachine_0.LocomotionStateMachine.AnimStateNode_0.EmoteLoops
  nodes [
   AnimGraphNode_ChooserPlayer_0 = Chooser Player: PRA_Companion_SpecialIdle!() -> (Pose:Pose Link Structure) {
    is=title AnimGraphNode_ChooserPlayer
    chooser {
     _structType=/Script/ProxyTable.LookupProxy
     proxy=/CosmeticCompanions/Chooser/PRA_Companion_SpecialIdle
     proxyTable {
      _structType=/Script/ProxyTable.ProxyTableContextProperty
      binding {contextIndex=0, displayName=Chooser Tables, isBoundToRoot=false, propertyBindingChain [ChooserTables]}
     }
    }
    chooserContextDefinition [
     {
      _structType=/Script/Chooser.ContextObjectTypeClass
      class=&/CrescentNail_AnimBP.CrescentNail_AnimBP_C
      direction=ReadWrite
     }
     {struct=/Script/Chooser.ChooserPlayerSettings, _structType=/Script/Chooser.ContextObjectTypeStruct, direction=Read}
    ]
    defaultSettings {blendOption=Linear}
   }
   Output Animation Pose(Result=@AnimGraphNode_ChooserPlayer_0)
  ]
 }
 AnimationBlendStackGraph_0 {
  path=AnimGraph.AnimGraphNode_StateMachine_0.LocomotionStateMachine.AnimStateNode_0.EmoteLoops.AnimGraphNode_ChooserPlayer_0.AnimationBlendStackGraph_0
  nodes [Output Pose!(Result:Pose Link Structure=Blend Stack Input!() -> (Pose:Pose Link Structure) {is=title AnimGraphNode_BlendStackInput}) {is=title AnimGraphNode_BlendStackResult}]
 }
 TransitionToDefaultIdle {
  path=AnimGraph.AnimGraphNode_StateMachine_0.LocomotionStateMachine.AnimStateNode_1.TransitionToDefaultIdle
  nodes [
   AnimGraphNode_ChooserPlayer_0 = Chooser Player: PRA_Companion_ReactToIdle!() -> (Pose:Pose Link Structure) {
    is=title AnimGraphNode_ChooserPlayer
    chooser {
     _structType=&LookupProxy
     proxy=/CosmeticCompanions/Chooser/PRA_Companion_ReactToIdle
     proxyTable {
      _structType=&ProxyTableContextProperty
      binding {contextIndex=0, displayName=Chooser Tables, isBoundToRoot=false, propertyBindingChain [ChooserTables]}
     }
    }
    chooserContextDefinition [
     {
      _structType=&ContextObjectTypeClass
      class=&CrescentNail_AnimBP_C
      direction=ReadWrite
     }
     {struct=&ChooserPlayerSettings, _structType=&ContextObjectTypeStruct, direction=Read}
    ]
    defaultSettings {blendOption=Linear}
   }
   Output Animation Pose(Result=@AnimGraphNode_ChooserPlayer_0) {
    node {stateEntryFunction {functionName=OnStateEntry_0}, stateExitFunction {functionName=OnStateExit_0}}
    stateEntryFunction {bSelfContext=true, memberGuid=#1, memberName=OnStateEntry_0}
    stateExitFunction {bSelfContext=true, memberGuid=#2, memberName=OnStateExit_0}
   }
  ]
 }
 AnimationBlendStackGraph_0 {
  path=AnimGraph.AnimGraphNode_StateMachine_0.LocomotionStateMachine.AnimStateNode_1.TransitionToDefaultIdle.AnimGraphNode_ChooserPlayer_0.AnimationBlendStackGraph_0
  nodes [Output Pose(Result=Blend Stack Input())]
 }
 LocoStop {
  path=AnimGraph.AnimGraphNode_StateMachine_0.LocomotionStateMachine.AnimStateNode_2.LocoStop
  nodes [
   AnimGraphNode_StateMachine_0 = LocomotionStop() {is=title AnimGraphNode_StateMachine, bAllowConduitEntryStates=true}
   AnimGraphNode_ChooserPlayer_0 = Chooser Player: PRA_Companion_Stops!(
    DefaultSettings_bMirror:bool=False
    DefaultSettings_StartTime:float=Get StartStopPosition() -> (StartStopPosition:float)
    DefaultSettings_bForceLooping:bool=False
    DefaultSettings_PlaybackRate:float=1
    DefaultSettings_CurveOverrides:Anim Curve Override List Structure=(Values=,Hash=0)
    DefaultSettings_BlendTime:float=0.2
    DefaultSettings_BlendProfile:Blend Profile Object Reference
    DefaultSettings_BlendOption:EAlphaBlendOption Enum=HermiteCubic
    DefaultSettings_bUseInertialBlend:bool=False
    DefaultSettings_bForceBlendTo:bool=False
    DefaultSettings_MinDeltaTimeToForceBlendTo:float=0.1
   ) -> (Pose:Pose Link Structure) {
    is=title AnimGraphNode_ChooserPlayer
    chooser {
     _structType=&LookupProxy
     proxy=/CosmeticCompanions/Chooser/PRA_Companion_Stops
     proxyTable {
      _structType=&ProxyTableContextProperty
      binding {contextIndex=0, displayName=Chooser Tables, isBoundToRoot=false, propertyBindingChain [ChooserTables]}
     }
    }
    chooserContextDefinition [
     {
      _structType=&ContextObjectTypeClass
      class=&CrescentNail_AnimBP_C
      direction=ReadWrite
     }
     {struct=&ChooserPlayerSettings, _structType=&ContextObjectTypeStruct, direction=Read}
    ]
    defaultSettings {blendOption=Linear}
   }
   Output Animation Pose(Result=@AnimGraphNode_ChooserPlayer_0)
  ]
 }
 AnimationBlendStackGraph_0 {
  path=AnimGraph.AnimGraphNode_StateMachine_0.LocomotionStateMachine.AnimStateNode_2.LocoStop.AnimGraphNode_ChooserPlayer_0.AnimationBlendStackGraph_0
  nodes [Output Pose(Result=Blend Stack Input())]
 }
 LocomotionStop {
  entry=Conduit
  path=AnimGraph.AnimGraphNode_StateMachine_0.LocomotionStateMachine.AnimStateNode_2.LocoStop.AnimGraphNode_StateMachine_0.LocomotionStop
  ids=id
  nodes [
   AnimStateConduitNode_0 = Conduit()
   AnimStateNode_0 = WalkStopFootL() {is=title AnimStateNode}
   AnimStateNode_1 = RunStopFootL() {is=title AnimStateNode}
   AnimStateNode_2 = WalkStopFootR() {is=title AnimStateNode}
   AnimStateNode_3 = RunStopFootR() {is=title AnimStateNode}
   AnimStateTransitionNode_0 = AnimStateTransitionNode() {logicType=TLT_Inertialization, priorityOrder=2}
   AnimStateTransitionNode_1 = AnimStateTransitionNode() {logicType=TLT_Inertialization, priorityOrder=4}
   AnimStateTransitionNode_2 = AnimStateTransitionNode()
   AnimStateTransitionNode_3 = AnimStateTransitionNode() {priorityOrder=3}
  ]
  links [
   {
    dst=RunStopFootL
    rule=AnimGraph.AnimGraphNode_StateMachine_0.LocomotionStateMachine.AnimStateNode_2.LocoStop.AnimGraphNode_StateMachine_0.LocomotionStop.AnimStateTransitionNode_1.Transition
    src=Conduit
    via=AnimStateTransitionNode_1
   }
   {
    dst=RunStopFootR
    rule=AnimGraph.AnimGraphNode_StateMachine_0.LocomotionStateMachine.AnimStateNode_2.LocoStop.AnimGraphNode_StateMachine_0.LocomotionStop.AnimStateTransitionNode_3.Transition
    src=Conduit
    via=AnimStateTransitionNode_3
   }
   {
    dst=WalkStopFootL
    rule=AnimGraph.AnimGraphNode_StateMachine_0.LocomotionStateMachine.AnimStateNode_2.LocoStop.AnimGraphNode_StateMachine_0.LocomotionStop.AnimStateTransitionNode_0.Transition
    src=Conduit
    via=AnimStateTransitionNode_0
   }
   {
    dst=WalkStopFootR
    rule=AnimGraph.AnimGraphNode_StateMachine_0.LocomotionStateMachine.AnimStateNode_2.LocoStop.AnimGraphNode_StateMachine_0.LocomotionStop.AnimStateTransitionNode_2.Transition
    src=Conduit
    via=AnimStateTransitionNode_2
   }
  ]
 }
 Conduit {
  path=AnimGraph.AnimGraphNode_StateMachine_0.LocomotionStateMachine.AnimStateNode_2.LocoStop.AnimGraphNode_StateMachine_0.LocomotionStop.AnimStateConduitNode_0.Conduit
  nodes [Result(bCanEnterTransition=true)]
 }
 WalkStopFootL {
  path=AnimGraph.AnimGraphNode_StateMachine_0.LocomotionStateMachine.AnimStateNode_2.LocoStop.AnimGraphNode_StateMachine_0.LocomotionStop.AnimStateNode_0.WalkStopFootL
  nodes [Output Animation Pose(Result=Sequence Player 'CrescentNail_COR_Loco_StopLight_L'!(StartPosition:float=Get StartStopPosition()) -> (Pose) {bLoopAnimation=false, sequence=/CosmeticCompanions/Assets/Biped/CrescentNail/Animation/Core/CrescentNail_COR_Loco_StopLight_L})]
 }
 RunStopFootL {
  path=AnimGraph.AnimGraphNode_StateMachine_0.LocomotionStateMachine.AnimStateNode_2.LocoStop.AnimGraphNode_StateMachine_0.LocomotionStop.AnimStateNode_1.RunStopFootL
  nodes [Output Animation Pose(Result=Sequence Player 'CrescentNail_COR_Loco_StopHeavy_L'!(StartPosition:float=Get StartStopPosition()) -> (Pose) {bLoopAnimation=false, sequence=/CosmeticCompanions/Assets/Biped/CrescentNail/Animation/Core/CrescentNail_COR_Loco_StopHeavy_L})]
 }
 WalkStopFootR {
  path=AnimGraph.AnimGraphNode_StateMachine_0.LocomotionStateMachine.AnimStateNode_2.LocoStop.AnimGraphNode_StateMachine_0.LocomotionStop.AnimStateNode_2.WalkStopFootR
  nodes [Output Animation Pose(Result=Sequence Player 'CrescentNail_COR_Loco_StopLight_R'!(StartPosition:float=Get StartStopPosition()) -> (Pose) {bLoopAnimation=false, sequence=/CosmeticCompanions/Assets/Biped/CrescentNail/Animation/Core/CrescentNail_COR_Loco_StopLight_R})]
 }
 RunStopFootR {
  path=AnimGraph.AnimGraphNode_StateMachine_0.LocomotionStateMachine.AnimStateNode_2.LocoStop.AnimGraphNode_StateMachine_0.LocomotionStop.AnimStateNode_3.RunStopFootR
  nodes [Output Animation Pose(Result=Sequence Player 'CrescentNail_COR_Loco_StopHeavy_R'!(StartPosition:float=Get StartStopPosition()) -> (Pose) {bLoopAnimation=false, sequence=/CosmeticCompanions/Assets/Biped/CrescentNail/Animation/Core/CrescentNail_COR_Loco_StopHeavy_R})]
 }
 Transition {
  path=AnimGraph.AnimGraphNode_StateMachine_0.LocomotionStateMachine.AnimStateNode_2.LocoStop.AnimGraphNode_StateMachine_0.LocomotionStop.AnimStateTransitionNode_0.Transition
  nodes [Result(bCanEnterTransition="float > float"(A=GetCurveValue(CurveName=walk), B=0.5))]
 }
 Transition {
  path=AnimGraph.AnimGraphNode_StateMachine_0.LocomotionStateMachine.AnimStateNode_2.LocoStop.AnimGraphNode_StateMachine_0.LocomotionStop.AnimStateTransitionNode_1.Transition
  nodes [Result(bCanEnterTransition=true)]
 }
 Transition {
  path=AnimGraph.AnimGraphNode_StateMachine_0.LocomotionStateMachine.AnimStateNode_2.LocoStop.AnimGraphNode_StateMachine_0.LocomotionStop.AnimStateTransitionNode_2.Transition
  nodes [
   n1 = GetCurveValue(CurveName=footphase)
   n2 = AND Boolean(A="float < float"(A=@n1, B=2.5), B="float >= float"(A=@n1, B=1.5))
   n3 = "float > float"(A=@n1, B=3.5)
   n4 = GetCurveValue(CurveName=walk)
   Result(bCanEnterTransition=AND Boolean(A=OR Boolean(A=@n3, B=@n2), B="float > float"(A=@n4, B=0.5)))
  ]
 }
 Transition {
  path=AnimGraph.AnimGraphNode_StateMachine_0.LocomotionStateMachine.AnimStateNode_2.LocoStop.AnimGraphNode_StateMachine_0.LocomotionStop.AnimStateTransitionNode_3.Transition
  nodes [
   n1 = GetCurveValue(CurveName=footphase)
   n2 = "float < float"(A=@n1, B=2.5)
   n3 = "float >= float"(A=@n1, B=1.5)
   Result(bCanEnterTransition=OR Boolean(A="float > float"(A=@n1, B=3.5), B=AND Boolean(A=@n2, B=@n3)))
  ]
 }
 Locomotion {
  path=AnimGraph.AnimGraphNode_StateMachine_0.LocomotionStateMachine.AnimStateNode_4.Locomotion
  nodes [
   n1 = Sequence Player 'CrescentNail_COR_Loco_Walk'() {groupName=Locomotion, method=SyncGroup, sequence=/CosmeticCompanions/Assets/Biped/CrescentNail/Animation/Core/CrescentNail_COR_Loco_Walk}
   n2 = Sequence Player 'CrescentNail_COR_Loco_Jog'() {groupName=Locomotion, method=SyncGroup, sequence=/CosmeticCompanions/Assets/Biped/CrescentNail/Animation/Core/CrescentNail_COR_Loco_Jog}
   n3 = Sequence Player 'CrescentNail_COR_Loco_Run'() {groupName=Locomotion, method=SyncGroup, sequence=/CosmeticCompanions/Assets/Biped/CrescentNail/Animation/Core/CrescentNail_COR_Loco_Run}
   n4 = Blend Poses by bool(
    BlendPose_0=@n3
    BlendPose_1=Blend Poses by bool(
     BlendPose_0=@n2
     BlendPose_1=Modify Curve(SourcePose=@n1, CurveValues_0=1, Alpha=1) {curveNames [Walk], curveValues [0]}
     BlendTime_0=0.2
     BlendTime_1=0.2
     bActiveValue=Get bIsLocoJog() -> (bIsLocoJog:bool)
    ) {blendType=Sinusoidal, transitionType=Inertialization}
    BlendTime_0=0.2
    BlendTime_1=0.2
    bActiveValue=Get bIsLocoRun() -> (bIsLocoRun:bool)
   ) {blendType=Sinusoidal, transitionType=Inertialization}
   n5 = Sequence Player 'CrescentNail_COR_Loco_Sprint'() {groupName=Locomotion, method=SyncGroup, sequence=/CosmeticCompanions/Assets/Biped/CrescentNail/Animation/Core/CrescentNail_COR_Loco_Sprint}
   n6 = Get bIsSprinting()
   Output Animation Pose(Result=Modify Curve(
     SourcePose=Blend Poses by bool(BlendPose_0=@n5, BlendPose_1=@n4, BlendTime_0=0.2, BlendTime_1=0.2, bActiveValue=@n6) {blendType=Sinusoidal, transitionType=Inertialization}
     CurveValues_0=1
     Alpha=1
    ) {curveNames [Locomotion], curveValues [0]})
  ]
 }
 Idle {
  path=AnimGraph.AnimGraphNode_StateMachine_0.LocomotionStateMachine.AnimStateNode_5.Idle
  nodes [
   AnimGraphNode_StateMachine_0 = Idle() {is=title AnimGraphNode_StateMachine, bAllowConduitEntryStates=true}
   Output Animation Pose(Result=@AnimGraphNode_StateMachine_0)
  ]
 }
 Idle {
  entry=Conduit
  path=AnimGraph.AnimGraphNode_StateMachine_0.LocomotionStateMachine.AnimStateNode_5.Idle.AnimGraphNode_StateMachine_0.Idle
  ids=id
  nodes [
   Alias() {is=title AnimStateAliasNode, also {aliases [SitPose,StandToSit]}}
   AnimStateConduitNode_0 = Conduit()
   AnimStateNode_0 = StandToSit() {is=title AnimStateNode}
   AnimStateNode_1 = SitPose() {is=title AnimStateNode}
   AnimStateNode_2 = StandingPose() {is=title AnimStateNode}
   AnimStateNode_3 = SitToStand() {is=title AnimStateNode}
   AnimStateTransitionNode_0 = AnimStateTransitionNode() {bAutomaticRuleBasedOnSequencePlayerInState=true, logicType=TLT_Inertialization}
   AnimStateTransitionNode_1 = AnimStateTransitionNode() {logicType=TLT_Inertialization, priorityOrder=2}
   AnimStateTransitionNode_2 = AnimStateTransitionNode() {logicType=TLT_Inertialization, priorityOrder=100}
   AnimStateTransitionNode_3 = AnimStateTransitionNode() {logicType=TLT_Inertialization}
   AnimStateTransitionNode_4 = AnimStateTransitionNode()
   AnimStateTransitionNode_6 = AnimStateTransitionNode() {bAutomaticRuleBasedOnSequencePlayerInState=true, crossfadeDuration=0.3, logicType=TLT_Inertialization}
   AnimStateTransitionNode_8 = AnimStateTransitionNode() {logicType=TLT_Inertialization}
  ]
  links [
   {
    dst=SitToStand
    rule=AnimGraph.AnimGraphNode_StateMachine_0.LocomotionStateMachine.AnimStateNode_5.Idle.AnimGraphNode_StateMachine_0.Idle.AnimStateTransitionNode_8.AnimationTransitionGraph_0
    src=Alias
    via=AnimStateTransitionNode_8
   }
   {
    dst=SitPose
    rule=AnimGraph.AnimGraphNode_StateMachine_0.LocomotionStateMachine.AnimStateNode_5.Idle.AnimGraphNode_StateMachine_0.Idle.AnimStateTransitionNode_3.AnimationTransitionGraph_0
    src=Conduit
    via=AnimStateTransitionNode_3
   }
   {
    dst=StandToSit
    rule=AnimGraph.AnimGraphNode_StateMachine_0.LocomotionStateMachine.AnimStateNode_5.Idle.AnimGraphNode_StateMachine_0.Idle.AnimStateTransitionNode_1.AnimationTransitionGraph_0
    src=Conduit
    via=AnimStateTransitionNode_1
   }
   {
    dst=StandingPose
    rule=AnimGraph.AnimGraphNode_StateMachine_0.LocomotionStateMachine.AnimStateNode_5.Idle.AnimGraphNode_StateMachine_0.Idle.AnimStateTransitionNode_2.AnimationTransitionGraph_0
    src=Conduit
    via=AnimStateTransitionNode_2
   }
   {
    dst=StandingPose
    rule=AnimGraph.AnimGraphNode_StateMachine_0.LocomotionStateMachine.AnimStateNode_5.Idle.AnimGraphNode_StateMachine_0.Idle.AnimStateTransitionNode_6.AnimationTransitionGraph_0
    src=SitToStand
    via=AnimStateTransitionNode_6
   }
   {
    dst=SitPose
    rule=AnimGraph.AnimGraphNode_StateMachine_0.LocomotionStateMachine.AnimStateNode_5.Idle.AnimGraphNode_StateMachine_0.Idle.AnimStateTransitionNode_0.AnimationTransitionGraph_0
    src=StandToSit
    via=AnimStateTransitionNode_0
   }
   {
    dst=Conduit
    rule=AnimGraph.AnimGraphNode_StateMachine_0.LocomotionStateMachine.AnimStateNode_5.Idle.AnimGraphNode_StateMachine_0.Idle.AnimStateTransitionNode_4.AnimationTransitionGraph_0
    src=StandingPose
    via=AnimStateTransitionNode_4
   }
  ]
 }
 Conduit {
  path=AnimGraph.AnimGraphNode_StateMachine_0.LocomotionStateMachine.AnimStateNode_5.Idle.AnimGraphNode_StateMachine_0.Idle.AnimStateConduitNode_0.Conduit
  nodes [Result(bCanEnterTransition=true)]
 }
 StandToSit {
  path=AnimGraph.AnimGraphNode_StateMachine_0.LocomotionStateMachine.AnimStateNode_5.Idle.AnimGraphNode_StateMachine_0.Idle.AnimStateNode_0.StandToSit
  nodes [Output Animation Pose(Result=Modify Curve(
     SourcePose=Sequence Player 'CrescentNail_COR_Sit_In'() {bLoopAnimation=false, sequence=/CosmeticCompanions/Assets/Biped/CrescentNail/Animation/Core/CrescentNail_COR_Sit_In}
     CurveValues_0=0
     Alpha=1
    ) {curveNames [AlphaCurve], curveValues [0]}) {
    node {stateEntryFunction {functionName=OnStateEntry}}
    stateEntryFunction {bSelfContext=true, memberGuid=#3, memberName=ToSitting_Entry}
   }]
 }
 SitPose {
  path=AnimGraph.AnimGraphNode_StateMachine_0.LocomotionStateMachine.AnimStateNode_5.Idle.AnimGraphNode_StateMachine_0.Idle.AnimStateNode_1.SitPose
  nodes [Output Animation Pose(Result=Modify Curve(
     SourcePose=Sequence Player 'CrescentNail_COR_Sit_Lp'() {sequence=/CosmeticCompanions/Assets/Biped/CrescentNail/Animation/Core/CrescentNail_COR_Sit_Lp}
     CurveValues_0=0
     Alpha=1
    ) {curveNames [AlphaCurve], curveValues [0]}) {
    node {stateEntryFunction {functionName=ToSitting_Entry}}
    stateEntryFunction {bSelfContext=true, memberGuid=#3, memberName=ToSitting_Entry}
   }]
 }
 StandingPose {
  path=AnimGraph.AnimGraphNode_StateMachine_0.LocomotionStateMachine.AnimStateNode_5.Idle.AnimGraphNode_StateMachine_0.Idle.AnimStateNode_2.StandingPose
  nodes [
   n1 = "Sequence Player (additive) 'CrescentNail_xDEV_Pose_Feathers_Spread'"() {bLoopAnimation=false, sequence=/CosmeticCompanions/Assets/Biped/CrescentNail/Animation/Developer/CrescentNail_xDEV_Pose_Feathers_Spread}
   Output Animation Pose(Result=Apply Additive(
     Base=Sequence Player 'CrescentNail_COR_Idle'() {sequence=/CosmeticCompanions/Assets/Biped/CrescentNail/Animation/Core/CrescentNail_COR_Idle, startPosition=2.18}
     Additive=@n1
    ) {alphaBoolBlend {blendOption=Linear}})
  ]
 }
 SitToStand {
  path=AnimGraph.AnimGraphNode_StateMachine_0.LocomotionStateMachine.AnimStateNode_5.Idle.AnimGraphNode_StateMachine_0.Idle.AnimStateNode_3.SitToStand
  nodes [Output Animation Pose(Result=Modify Curve(
     SourcePose=Sequence Player 'CrescentNail_COR_Sit_Out'() {bLoopAnimation=false, sequence=/CosmeticCompanions/Assets/Biped/CrescentNail/Animation/Core/CrescentNail_COR_Sit_Out}
     CurveValues_0=0
     Alpha=1
    ) {curveNames [AlphaCurve], curveValues [0]}) {
    node {stateExitFunction {functionName=OnStateExit}}
    stateExitFunction {bSelfContext=true, memberGuid=#4, memberName=OnStateExit}
   }]
 }
 AnimationTransitionGraph_0 {
  path=AnimGraph.AnimGraphNode_StateMachine_0.LocomotionStateMachine.AnimStateNode_5.Idle.AnimGraphNode_StateMachine_0.Idle.AnimStateTransitionNode_0.AnimationTransitionGraph_0
  nodes [Result()]
 }
 AnimationTransitionGraph_0 {
  path=AnimGraph.AnimGraphNode_StateMachine_0.LocomotionStateMachine.AnimStateNode_5.Idle.AnimGraphNode_StateMachine_0.Idle.AnimStateTransitionNode_1.AnimationTransitionGraph_0
  nodes [Result(bCanEnterTransition=Get bIsStandToSit() -> (bIsStandToSit:bool))]
 }
 AnimationTransitionGraph_0 {
  path=AnimGraph.AnimGraphNode_StateMachine_0.LocomotionStateMachine.AnimStateNode_5.Idle.AnimGraphNode_StateMachine_0.Idle.AnimStateTransitionNode_2.AnimationTransitionGraph_0
  nodes [Result(bCanEnterTransition=true)]
 }
 AnimationTransitionGraph_0 {
  path=AnimGraph.AnimGraphNode_StateMachine_0.LocomotionStateMachine.AnimStateNode_5.Idle.AnimGraphNode_StateMachine_0.Idle.AnimStateTransitionNode_3.AnimationTransitionGraph_0
  nodes [Result(bCanEnterTransition=Get bIsSitting() -> (bIsSitting:bool))]
 }
 AnimationTransitionGraph_0 {
  path=AnimGraph.AnimGraphNode_StateMachine_0.LocomotionStateMachine.AnimStateNode_5.Idle.AnimGraphNode_StateMachine_0.Idle.AnimStateTransitionNode_4.AnimationTransitionGraph_0
  nodes [Result(bCanEnterTransition=true)]
 }
 AnimationTransitionGraph_0 {
  path=AnimGraph.AnimGraphNode_StateMachine_0.LocomotionStateMachine.AnimStateNode_5.Idle.AnimGraphNode_StateMachine_0.Idle.AnimStateTransitionNode_6.AnimationTransitionGraph_0
  nodes [Result()]
 }
 AnimationTransitionGraph_0 {
  path=AnimGraph.AnimGraphNode_StateMachine_0.LocomotionStateMachine.AnimStateNode_5.Idle.AnimGraphNode_StateMachine_0.Idle.AnimStateTransitionNode_8.AnimationTransitionGraph_0
  nodes [Result(bCanEnterTransition=Get bIsSitToStand() -> (bIsSitToStand:bool))]
 }
 AnimationTransitionGraph_0 {path=AnimGraph.AnimGraphNode_StateMachine_0.LocomotionStateMachine.AnimStateTransitionNode_0.AnimationTransitionGraph_0, nodes [Result()]}
 AnimationTransitionGraph_0 {
  path=AnimGraph.AnimGraphNode_StateMachine_0.LocomotionStateMachine.AnimStateTransitionNode_1.AnimationTransitionGraph_0
  nodes [Result(bCanEnterTransition=true)]
 }
 AnimationTransitionGraph_0 {
  path=AnimGraph.AnimGraphNode_StateMachine_0.LocomotionStateMachine.AnimStateTransitionNode_10.AnimationTransitionGraph_0
  nodes [Result(bCanEnterTransition=Get bEnterEmoteLoops() -> (bEnterEmoteLoops:bool))]
 }
 AnimationTransitionGraph_0 {
  path=AnimGraph.AnimGraphNode_StateMachine_0.LocomotionStateMachine.AnimStateTransitionNode_11.AnimationTransitionGraph_0
  nodes [Result(bCanEnterTransition=NOT Boolean(A=Get bIsPlayingEmote()))]
 }
 AnimationTransitionGraph_0 {
  path=AnimGraph.AnimGraphNode_StateMachine_0.LocomotionStateMachine.AnimStateTransitionNode_12.AnimationTransitionGraph_0
  nodes [Result(bCanEnterTransition=Get bIsReactionHappyPlaying() -> (bIsReactionHappyPlaying:bool))]
 }
 AnimationTransitionGraph_0 {
  path=AnimGraph.AnimGraphNode_StateMachine_0.LocomotionStateMachine.AnimStateTransitionNode_13.AnimationTransitionGraph_0
  nodes [Result(bCanEnterTransition=Get bIsMoving() -> (bIsMoving:bool))]
 }
 AnimationTransitionGraph_0 {
  path=AnimGraph.AnimGraphNode_StateMachine_0.LocomotionStateMachine.AnimStateTransitionNode_14.AnimationTransitionGraph_0
  nodes [Result(bCanEnterTransition=Get bIsMoving())]
 }
 AnimationTransitionGraph_0 {path=AnimGraph.AnimGraphNode_StateMachine_0.LocomotionStateMachine.AnimStateTransitionNode_15.AnimationTransitionGraph_0, nodes [Result()]}
 AnimationTransitionGraph_0 {
  path=AnimGraph.AnimGraphNode_StateMachine_0.LocomotionStateMachine.AnimStateTransitionNode_16.AnimationTransitionGraph_0
  nodes [Result(bCanEnterTransition=Get bIsMoving())]
 }
 AnimationTransitionGraph_0 {
  path=AnimGraph.AnimGraphNode_StateMachine_0.LocomotionStateMachine.AnimStateTransitionNode_17.AnimationTransitionGraph_0
  nodes [Result(bCanEnterTransition=Get bIsMoving())]
 }
 AnimationTransitionGraph_0 {
  path=AnimGraph.AnimGraphNode_StateMachine_0.LocomotionStateMachine.AnimStateTransitionNode_2.AnimationTransitionGraph_0
  nodes [Result(bCanEnterTransition=true)]
 }
 AnimationTransitionGraph_0 {
  path=AnimGraph.AnimGraphNode_StateMachine_0.LocomotionStateMachine.AnimStateTransitionNode_4.AnimationTransitionGraph_0
  nodes [Result(bCanEnterTransition=NOT Boolean(A=Get bIsMoving()))]
 }
 Body Pose Additive {
  entry=Conduit
  path=AnimGraph.AnimGraphNode_StateMachine_1.Body Pose Additive
  ids=id
  nodes [
   AnimStateConduitNode_0 = Conduit()
   AnimStateNode_0 = BodyPoseChooser() {is=title AnimStateNode}
   AnimStateNode_2 = BodyDefaultPose() {is=title AnimStateNode}
   AnimStateTransitionNode_0 = AnimStateTransitionNode() {
    crossfadeDuration=0.4
    logicType=TLT_Inertialization
    blendProfileWrapper {blendProfile=&CrescentNail_Biped_Skeleton:FastHeadSlowBody}
   }
   AnimStateTransitionNode_2 = AnimStateTransitionNode() {
    crossfadeDuration=0.6
    logicType=TLT_Inertialization
    priorityOrder=100
    blendProfileWrapper {blendProfile=&CrescentNail_Biped_Skeleton:FastHeadSlowBody}
   }
   AnimStateTransitionNode_3 = AnimStateTransitionNode() {bAutomaticRuleBasedOnSequencePlayerInState=true}
   AnimStateTransitionNode_5 = AnimStateTransitionNode()
  ]
  links [
   {
    dst=Conduit
    rule=AnimGraph.AnimGraphNode_StateMachine_1.Body Pose Additive.AnimStateTransitionNode_5.Transition
    src=BodyDefaultPose
    via=AnimStateTransitionNode_5
   }
   {
    dst=Conduit
    rule=AnimGraph.AnimGraphNode_StateMachine_1.Body Pose Additive.AnimStateTransitionNode_3.Transition
    src=BodyPoseChooser
    via=AnimStateTransitionNode_3
   }
   {
    dst=BodyDefaultPose
    rule=AnimGraph.AnimGraphNode_StateMachine_1.Body Pose Additive.AnimStateTransitionNode_2.Transition
    src=Conduit
    via=AnimStateTransitionNode_2
   }
   {
    dst=BodyPoseChooser
    rule=AnimGraph.AnimGraphNode_StateMachine_1.Body Pose Additive.AnimStateTransitionNode_0.Transition
    src=Conduit
    via=AnimStateTransitionNode_0
   }
  ]
 }
 Conduit {path=AnimGraph.AnimGraphNode_StateMachine_1.Body Pose Additive.AnimStateConduitNode_0.Conduit, nodes [Result(bCanEnterTransition=true)]}
 BodyPoseChooser {
  path=AnimGraph.AnimGraphNode_StateMachine_1.Body Pose Additive.AnimStateNode_0.BodyPoseChooser
  nodes [
   AnimGraphNode_ChooserPlayer_0 = Chooser Player: PRA_Companion_BodyPose!() -> (Pose:Pose Link Structure) {
    is=title AnimGraphNode_ChooserPlayer
    chooser {
     _structType=&LookupProxy
     proxy=/CosmeticCompanions/Chooser/PRA_Companion_BodyPose
     proxyTable {
      _structType=&ProxyTableContextProperty
      binding {contextIndex=0, displayName=Chooser Tables, isBoundToRoot=false, propertyBindingChain [ChooserTables]}
     }
    }
    chooserContextDefinition [
     {
      _structType=&ContextObjectTypeClass
      class=&CrescentNail_AnimBP_C
      direction=ReadWrite
     }
     {struct=&ChooserPlayerSettings, _structType=&ContextObjectTypeStruct, direction=Read}
    ]
    defaultSettings {blendOption=Linear}
   }
   Output Animation Pose(Result=@AnimGraphNode_ChooserPlayer_0)
  ]
 }
 AnimationBlendStackGraph_0 {
  path=AnimGraph.AnimGraphNode_StateMachine_1.Body Pose Additive.AnimStateNode_0.BodyPoseChooser.AnimGraphNode_ChooserPlayer_0.AnimationBlendStackGraph_0
  nodes [Output Pose(Result=Blend Stack Input())]
 }
 BodyDefaultPose {
  path=AnimGraph.AnimGraphNode_StateMachine_1.Body Pose Additive.AnimStateNode_2.BodyDefaultPose
  nodes [Output Animation Pose(Result=Additive Identity Pose())]
 }
 Transition {
  path=AnimGraph.AnimGraphNode_StateMachine_1.Body Pose Additive.AnimStateTransitionNode_0.Transition
  nodes [Result(bCanEnterTransition=Break Companion Random Number Timer!(CompanionRandomNumberTimer:"Companion Random Number Timer Structure (by ref)"=Get AnticsTimer() -> (AnticsTimer:Companion Random Number Timer Structure)) -> (ValueInt:int, ValueFloat:float=~, bJustChanged:bool=~) {is=title K2Node_BreakStruct}.bJustChanged)]
 }
 Transition {
  path=AnimGraph.AnimGraphNode_StateMachine_1.Body Pose Additive.AnimStateTransitionNode_2.Transition
  nodes [Result(bCanEnterTransition=true)]
 }
 Transition {path=AnimGraph.AnimGraphNode_StateMachine_1.Body Pose Additive.AnimStateTransitionNode_3.Transition, nodes [Result()]}
 Transition {
  path=AnimGraph.AnimGraphNode_StateMachine_1.Body Pose Additive.AnimStateTransitionNode_5.Transition
  nodes [Result(bCanEnterTransition=true)]
 }
 BodyMoodAdditive {
  entry=Conduit
  path=AnimGraph.AnimGraphNode_StateMachine_3.BodyMoodAdditive
  ids=id
  nodes [
   AnimStateConduitNode_0 = Conduit()
   AnimStateNode_0 = Worried() {is=title AnimStateNode}
   AnimStateNode_1 = Default() {is=title AnimStateNode}
   AnimStateTransitionNode_0 = AnimStateTransitionNode() {logicType=TLT_Inertialization}
   AnimStateTransitionNode_1 = AnimStateTransitionNode() {logicType=TLT_Inertialization, priorityOrder=100}
   AnimStateTransitionNode_2 = AnimStateTransitionNode()
   AnimStateTransitionNode_3 = AnimStateTransitionNode()
  ]
  links [
   {
    dst=Default
    rule=AnimGraph.AnimGraphNode_StateMachine_3.BodyMoodAdditive.AnimStateTransitionNode_1.Transition
    src=Conduit
    via=AnimStateTransitionNode_1
   }
   {
    dst=Worried
    rule=AnimGraph.AnimGraphNode_StateMachine_3.BodyMoodAdditive.AnimStateTransitionNode_0.Transition
    src=Conduit
    via=AnimStateTransitionNode_0
   }
   {
    dst=Conduit
    rule=AnimGraph.AnimGraphNode_StateMachine_3.BodyMoodAdditive.AnimStateTransitionNode_2.Transition
    src=Default
    via=AnimStateTransitionNode_2
   }
   {
    dst=Conduit
    rule=AnimGraph.AnimGraphNode_StateMachine_3.BodyMoodAdditive.AnimStateTransitionNode_3.Transition
    src=Worried
    via=AnimStateTransitionNode_3
   }
  ]
 }
 Conduit {path=AnimGraph.AnimGraphNode_StateMachine_3.BodyMoodAdditive.AnimStateConduitNode_0.Conduit, nodes [Result(bCanEnterTransition=true)]}
 Worried {
  path=AnimGraph.AnimGraphNode_StateMachine_3.BodyMoodAdditive.AnimStateNode_0.Worried
  nodes [Output Animation Pose(Result="Sequence Player (additive) 'CrescentNail_COR_Loco_Worried_ADD'"() {bLoopAnimation=false, sequence=/CosmeticCompanions/Assets/Biped/CrescentNail/Animation/Core/CrescentNail_COR_Loco_Worried_ADD})]
 }
 Default {
  path=AnimGraph.AnimGraphNode_StateMachine_3.BodyMoodAdditive.AnimStateNode_1.Default
  nodes [Output Animation Pose(Result=Additive Identity Pose())]
 }
 Transition {
  path=AnimGraph.AnimGraphNode_StateMachine_3.BodyMoodAdditive.AnimStateTransitionNode_0.Transition
  nodes [Result(bCanEnterTransition=Get bIsMoodWorried() -> (bIsMoodWorried:bool))]
 }
 Transition {path=AnimGraph.AnimGraphNode_StateMachine_3.BodyMoodAdditive.AnimStateTransitionNode_1.Transition, nodes [Result(bCanEnterTransition=true)]}
 Transition {path=AnimGraph.AnimGraphNode_StateMachine_3.BodyMoodAdditive.AnimStateTransitionNode_2.Transition, nodes [Result(bCanEnterTransition=true)]}
 Transition {path=AnimGraph.AnimGraphNode_StateMachine_3.BodyMoodAdditive.AnimStateTransitionNode_3.Transition, nodes [Result(bCanEnterTransition=true)]}
 LowFidelityStateMachine {
  entry=Conduit
  path=AnimGraph.AnimGraphNode_StateMachine_4.LowFidelityStateMachine
  ids=id
  nodes [
   AnimStateConduitNode_0 = Conduit()
   AnimStateNode_0 = EmoteLoops()
   AnimStateNode_1 = TransitionToDefaultIdle()
   AnimStateNode_4 = Locomotion()
   AnimStateNode_5 = Idle() {is=title AnimStateNode, bAlwaysResetOnEntry=true}
   AnimStateTransitionNode_1 = AnimStateTransitionNode() {crossfadeDuration=0.35, priorityOrder=100}
   AnimStateTransitionNode_10 = AnimStateTransitionNode() {logicType=TLT_Inertialization}
   AnimStateTransitionNode_11 = AnimStateTransitionNode() {logicType=TLT_Inertialization}
   AnimStateTransitionNode_12 = AnimStateTransitionNode() {logicType=TLT_Inertialization, priorityOrder=2}
   AnimStateTransitionNode_13 = AnimStateTransitionNode()
   AnimStateTransitionNode_14 = AnimStateTransitionNode()
   AnimStateTransitionNode_15 = AnimStateTransitionNode() {bAutomaticRuleBasedOnSequencePlayerInState=true}
   AnimStateTransitionNode_16 = AnimStateTransitionNode() {
    crossfadeDuration=0.3
    logicType=TLT_Inertialization
    blendProfileWrapper {blendProfile=&CrescentNail_Biped_Skeleton:FastFeetSlowHead}
   }
   AnimStateTransitionNode_2 = AnimStateTransitionNode()
   AnimStateTransitionNode_22 = AnimStateTransitionNode()
  ]
  links [
   {
    dst=EmoteLoops
    rule=AnimGraph.AnimGraphNode_StateMachine_4.LowFidelityStateMachine.AnimStateTransitionNode_10.AnimationTransitionGraph_0
    src=Conduit
    via=AnimStateTransitionNode_10
   }
   {
    dst=Idle
    rule=AnimGraph.AnimGraphNode_StateMachine_4.LowFidelityStateMachine.AnimStateTransitionNode_1.AnimationTransitionGraph_0
    src=Conduit
    via=AnimStateTransitionNode_1
   }
   {
    dst=Locomotion
    rule=AnimGraph.AnimGraphNode_StateMachine_4.LowFidelityStateMachine.AnimStateTransitionNode_16.AnimationTransitionGraph_0
    src=Conduit
    via=AnimStateTransitionNode_16
   }
   {
    dst=TransitionToDefaultIdle
    rule=AnimGraph.AnimGraphNode_StateMachine_4.LowFidelityStateMachine.AnimStateTransitionNode_12.AnimationTransitionGraph_0
    src=Conduit
    via=AnimStateTransitionNode_12
   }
   {
    dst=Conduit
    rule=AnimGraph.AnimGraphNode_StateMachine_4.LowFidelityStateMachine.AnimStateTransitionNode_14.AnimationTransitionGraph_0
    src=EmoteLoops
    via=AnimStateTransitionNode_14
   }
   {
    dst=TransitionToDefaultIdle
    rule=AnimGraph.AnimGraphNode_StateMachine_4.LowFidelityStateMachine.AnimStateTransitionNode_11.AnimationTransitionGraph_0
    src=EmoteLoops
    via=AnimStateTransitionNode_11
   }
   {
    dst=Conduit
    rule=AnimGraph.AnimGraphNode_StateMachine_4.LowFidelityStateMachine.AnimStateTransitionNode_2.AnimationTransitionGraph_0
    src=Idle
    via=AnimStateTransitionNode_2
   }
   {
    dst=Conduit
    rule=AnimGraph.AnimGraphNode_StateMachine_4.LowFidelityStateMachine.AnimStateTransitionNode_22.Transition
    src=Locomotion
    via=AnimStateTransitionNode_22
   }
   {
    dst=Conduit
    rule=AnimGraph.AnimGraphNode_StateMachine_4.LowFidelityStateMachine.AnimStateTransitionNode_13.AnimationTransitionGraph_0
    src=TransitionToDefaultIdle
    via=AnimStateTransitionNode_13
   }
   {
    dst=Conduit
    rule=AnimGraph.AnimGraphNode_StateMachine_4.LowFidelityStateMachine.AnimStateTransitionNode_15.AnimationTransitionGraph_0
    src=TransitionToDefaultIdle
    via=AnimStateTransitionNode_15
   }
  ]
 }
 Conduit {path=AnimGraph.AnimGraphNode_StateMachine_4.LowFidelityStateMachine.AnimStateConduitNode_0.Conduit, nodes [Result(bCanEnterTransition=true)]}
 EmoteLoops {
  path=AnimGraph.AnimGraphNode_StateMachine_4.LowFidelityStateMachine.AnimStateNode_0.EmoteLoops
  nodes [
   AnimGraphNode_ChooserPlayer_0 = Chooser Player: PRA_Companion_SpecialIdle() {
    chooser {
     _structType=&LookupProxy
     proxy=&PRA_Companion_SpecialIdle
     proxyTable {
      _structType=&ProxyTableContextProperty
      binding {contextIndex=0, displayName=Chooser Tables, isBoundToRoot=false, propertyBindingChain [ChooserTables]}
     }
    }
    chooserContextDefinition [
     {
      _structType=&ContextObjectTypeClass
      class=&CrescentNail_AnimBP_C
      direction=ReadWrite
     }
     {struct=&ChooserPlayerSettings, _structType=&ContextObjectTypeStruct, direction=Read}
    ]
    defaultSettings {blendOption=Linear}
   }
   Output Animation Pose(Result=@AnimGraphNode_ChooserPlayer_0)
  ]
 }
 AnimationBlendStackGraph_0 {
  path=AnimGraph.AnimGraphNode_StateMachine_4.LowFidelityStateMachine.AnimStateNode_0.EmoteLoops.AnimGraphNode_ChooserPlayer_0.AnimationBlendStackGraph_0
  nodes [Output Pose(Result=Blend Stack Input())]
 }
 TransitionToDefaultIdle {
  path=AnimGraph.AnimGraphNode_StateMachine_4.LowFidelityStateMachine.AnimStateNode_1.TransitionToDefaultIdle
  nodes [
   AnimGraphNode_ChooserPlayer_0 = Chooser Player: PRA_Companion_ReactToIdle() {
    chooser {
     _structType=&LookupProxy
     proxy=&PRA_Companion_ReactToIdle
     proxyTable {
      _structType=&ProxyTableContextProperty
      binding {contextIndex=0, displayName=Chooser Tables, isBoundToRoot=false, propertyBindingChain [ChooserTables]}
     }
    }
    chooserContextDefinition [
     {
      _structType=&ContextObjectTypeClass
      class=&CrescentNail_AnimBP_C
      direction=ReadWrite
     }
     {struct=&ChooserPlayerSettings, _structType=&ContextObjectTypeStruct, direction=Read}
    ]
    defaultSettings {blendOption=Linear}
   }
   Output Animation Pose(Result=@AnimGraphNode_ChooserPlayer_0) {
    node {stateEntryFunction {functionName=OnStateEntry_0}, stateExitFunction {functionName=OnStateExit_0}}
    stateEntryFunction {bSelfContext=true, memberGuid=#1, memberName=OnStateEntry_0}
    stateExitFunction {bSelfContext=true, memberGuid=#2, memberName=OnStateExit_0}
   }
  ]
 }
 AnimationBlendStackGraph_0 {
  path=AnimGraph.AnimGraphNode_StateMachine_4.LowFidelityStateMachine.AnimStateNode_1.TransitionToDefaultIdle.AnimGraphNode_ChooserPlayer_0.AnimationBlendStackGraph_0
  nodes [Output Pose(Result=Blend Stack Input())]
 }
 Locomotion {
  path=AnimGraph.AnimGraphNode_StateMachine_4.LowFidelityStateMachine.AnimStateNode_4.Locomotion
  nodes [
   n1 = Sequence Player 'CrescentNail_COR_Loco_Walk'() {groupName=Locomotion, method=SyncGroup, sequence=&CrescentNail_COR_Loco_Walk}
   n2 = Sequence Player 'CrescentNail_COR_Loco_Jog'() {groupName=Locomotion, method=SyncGroup, sequence=&CrescentNail_COR_Loco_Jog}
   n3 = Sequence Player 'CrescentNail_COR_Loco_Run'() {groupName=Locomotion, method=SyncGroup, sequence=&CrescentNail_COR_Loco_Run}
   n4 = Blend Poses by bool(
    BlendPose_0=@n3
    BlendPose_1=Blend Poses by bool(
     BlendPose_0=@n2
     BlendPose_1=Modify Curve(SourcePose=@n1, CurveValues_0=1, Alpha=1) {curveNames [Walk], curveValues [0]}
     BlendTime_0=0.2
     BlendTime_1=0.2
     bActiveValue=Get bIsLocoJog()
    ) {blendType=Sinusoidal, transitionType=Inertialization}
    BlendTime_0=0.2
    BlendTime_1=0.2
    bActiveValue=Get bIsLocoRun()
   ) {blendType=Sinusoidal, transitionType=Inertialization}
   n5 = Sequence Player 'CrescentNail_COR_Loco_Sprint'() {groupName=Locomotion, method=SyncGroup, sequence=&CrescentNail_COR_Loco_Sprint}
   n6 = Get bIsSprinting()
   Output Animation Pose(Result=Modify Curve(
     SourcePose=Blend Poses by bool(BlendPose_0=@n5, BlendPose_1=@n4, BlendTime_0=0.2, BlendTime_1=0.2, bActiveValue=@n6) {blendType=Sinusoidal, transitionType=Inertialization}
     CurveValues_0=1
     Alpha=1
    ) {curveNames [Locomotion], curveValues [0]})
  ]
 }
 Idle {
  path=AnimGraph.AnimGraphNode_StateMachine_4.LowFidelityStateMachine.AnimStateNode_5.Idle
  nodes [
   n1 = "Sequence Player (additive) 'CrescentNail_xDEV_Pose_Feathers_Spread'"() {bLoopAnimation=false, sequence=&CrescentNail_xDEV_Pose_Feathers_Spread}
   Output Animation Pose(Result=Apply Additive(
     Base=Sequence Player 'CrescentNail_COR_Idle'() {sequence=&CrescentNail_COR_Idle, startPosition=2.18}
     Additive=@n1
    ) {alphaBoolBlend {blendOption=Linear}})
   Use cached pose 'CrescentNailIdle'()
  ]
 }
 AnimationTransitionGraph_0 {
  path=AnimGraph.AnimGraphNode_StateMachine_4.LowFidelityStateMachine.AnimStateTransitionNode_1.AnimationTransitionGraph_0
  nodes [Result(bCanEnterTransition=true)]
 }
 AnimationTransitionGraph_0 {
  path=AnimGraph.AnimGraphNode_StateMachine_4.LowFidelityStateMachine.AnimStateTransitionNode_10.AnimationTransitionGraph_0
  nodes [Result(bCanEnterTransition=Get bEnterEmoteLoops())]
 }
 AnimationTransitionGraph_0 {
  path=AnimGraph.AnimGraphNode_StateMachine_4.LowFidelityStateMachine.AnimStateTransitionNode_11.AnimationTransitionGraph_0
  nodes [Result(bCanEnterTransition=NOT Boolean(A=Get bIsPlayingEmote()))]
 }
 AnimationTransitionGraph_0 {
  path=AnimGraph.AnimGraphNode_StateMachine_4.LowFidelityStateMachine.AnimStateTransitionNode_12.AnimationTransitionGraph_0
  nodes [Result(bCanEnterTransition=Get bIsReactionHappyPlaying())]
 }
 AnimationTransitionGraph_0 {
  path=AnimGraph.AnimGraphNode_StateMachine_4.LowFidelityStateMachine.AnimStateTransitionNode_13.AnimationTransitionGraph_0
  nodes [Result(bCanEnterTransition=Get bIsMoving())]
 }
 AnimationTransitionGraph_0 {
  path=AnimGraph.AnimGraphNode_StateMachine_4.LowFidelityStateMachine.AnimStateTransitionNode_14.AnimationTransitionGraph_0
  nodes [Result(bCanEnterTransition=Get bIsMoving())]
 }
 AnimationTransitionGraph_0 {path=AnimGraph.AnimGraphNode_StateMachine_4.LowFidelityStateMachine.AnimStateTransitionNode_15.AnimationTransitionGraph_0, nodes [Result()]}
 AnimationTransitionGraph_0 {
  path=AnimGraph.AnimGraphNode_StateMachine_4.LowFidelityStateMachine.AnimStateTransitionNode_16.AnimationTransitionGraph_0
  nodes [Result(bCanEnterTransition=Get bIsMoving())]
 }
 AnimationTransitionGraph_0 {
  path=AnimGraph.AnimGraphNode_StateMachine_4.LowFidelityStateMachine.AnimStateTransitionNode_2.AnimationTransitionGraph_0
  nodes [Result(bCanEnterTransition=true)]
 }
 Transition {
  path=AnimGraph.AnimGraphNode_StateMachine_4.LowFidelityStateMachine.AnimStateTransitionNode_22.Transition
  nodes [Result(bCanEnterTransition=true)]
 }
 LocomotionStartAdditive {
  entry=IdentPose
  path=AnimGraph.AnimGraphNode_StateMachine_5.LocomotionStartAdditive
  ids=id
  nodes [
   AnimStateNode_0 = IdentPose() {is=title AnimStateNode}
   AnimStateNode_1 = Start() {is=title AnimStateNode}
   AnimStateTransitionNode_0 = AnimStateTransitionNode() {crossfadeDuration=0.1}
   AnimStateTransitionNode_1 = AnimStateTransitionNode() {bAutomaticRuleBasedOnSequencePlayerInState=true}
  ]
  links [
   {
    dst=Start
    rule=AnimGraph.AnimGraphNode_StateMachine_5.LocomotionStartAdditive.AnimStateTransitionNode_0.Transition
    src=IdentPose
    via=AnimStateTransitionNode_0
   }
   {
    dst=IdentPose
    rule=AnimGraph.AnimGraphNode_StateMachine_5.LocomotionStartAdditive.AnimStateTransitionNode_1.Transition
    src=Start
    via=AnimStateTransitionNode_1
   }
  ]
 }
 IdentPose {
  path=AnimGraph.AnimGraphNode_StateMachine_5.LocomotionStartAdditive.AnimStateNode_0.IdentPose
  nodes [Output Animation Pose(Result=Additive Identity Pose())]
 }
 Start {
  path=AnimGraph.AnimGraphNode_StateMachine_5.LocomotionStartAdditive.AnimStateNode_1.Start
  nodes [
   n1 = Layered blend per bone(
    BasePose="Sequence Player (additive) 'CrescentNail_COR_Loco_Start'"() {bLoopAnimation=false, sequence=/CosmeticCompanions/Assets/Biped/CrescentNail/Animation/Core/CrescentNail_COR_Loco_Start}
    BlendPoses_0=Additive Identity Pose()
   ) {
    blendMode=BlendMask
    blendMasks [&CrescentNail_Biped_Skeleton:NoLegsOrArms]
    layerSetup []
   }
   Output Animation Pose(Result=@n1)
  ]
 }
 Transition {
  path=AnimGraph.AnimGraphNode_StateMachine_5.LocomotionStartAdditive.AnimStateTransitionNode_0.Transition
  nodes [Result(bCanEnterTransition=Get bIdentPoseToStart() -> (bIdentPoseToStart:bool))]
 }
 Transition {path=AnimGraph.AnimGraphNode_StateMachine_5.LocomotionStartAdditive.AnimStateTransitionNode_1.Transition, nodes [Result()]}
 Head Antic {
  entry=Conduit
  path=AnimGraph.AnimGraphNode_StateMachine_6.Head Antic
  ids=id
  nodes [
   AnimStateConduitNode_0 = Conduit()
   AnimStateNode_0 = AnticDefaultPose() {is=title AnimStateNode}
   AnimStateNode_1 = AnticChooser() {is=title AnimStateNode}
   AnimStateTransitionNode_0 = AnimStateTransitionNode() {logicType=TLT_Inertialization, priorityOrder=100}
   AnimStateTransitionNode_1 = AnimStateTransitionNode() {logicType=TLT_Inertialization}
   AnimStateTransitionNode_2 = AnimStateTransitionNode() {bAutomaticRuleBasedOnSequencePlayerInState=true}
   AnimStateTransitionNode_5 = AnimStateTransitionNode()
  ]
  links [
   {
    dst=Conduit
    rule=AnimGraph.AnimGraphNode_StateMachine_6.Head Antic.AnimStateTransitionNode_2.Transition
    src=AnticChooser
    via=AnimStateTransitionNode_2
   }
   {
    dst=Conduit
    rule=AnimGraph.AnimGraphNode_StateMachine_6.Head Antic.AnimStateTransitionNode_5.Transition
    src=AnticDefaultPose
    via=AnimStateTransitionNode_5
   }
   {
    dst=AnticChooser
    rule=AnimGraph.AnimGraphNode_StateMachine_6.Head Antic.AnimStateTransitionNode_1.Transition
    src=Conduit
    via=AnimStateTransitionNode_1
   }
   {
    dst=AnticDefaultPose
    rule=AnimGraph.AnimGraphNode_StateMachine_6.Head Antic.AnimStateTransitionNode_0.Transition
    src=Conduit
    via=AnimStateTransitionNode_0
   }
  ]
 }
 Conduit {path=AnimGraph.AnimGraphNode_StateMachine_6.Head Antic.AnimStateConduitNode_0.Conduit, nodes [Result(bCanEnterTransition=true)]}
 AnticDefaultPose {
  path=AnimGraph.AnimGraphNode_StateMachine_6.Head Antic.AnimStateNode_0.AnticDefaultPose
  nodes [Output Animation Pose(Result=Additive Identity Pose())]
 }
 AnticChooser {
  path=AnimGraph.AnimGraphNode_StateMachine_6.Head Antic.AnimStateNode_1.AnticChooser
  nodes [Output Animation Pose(Result=Sequence Player!(Sequence:Anim Sequence Base Object Reference) -> (Pose) {bLoopAnimation=false})]
 }
 Transition {path=AnimGraph.AnimGraphNode_StateMachine_6.Head Antic.AnimStateTransitionNode_0.Transition, nodes [Result(bCanEnterTransition=true)]}
 Transition {
  path=AnimGraph.AnimGraphNode_StateMachine_6.Head Antic.AnimStateTransitionNode_1.Transition
  nodes [
   Result(bCanEnterTransition=Get!() -> (AnticSequenceInfo_AnticSequenceToPlay:Anim Sequence Object Reference, AnticSequenceInfo_bAnticSequenceReady:bool=~).AnticSequenceInfo_bAnticSequenceReady)
   Break Companion Random Number Timer(CompanionRandomNumberTimer=Get AnticsTimer()) -> (ValueFloat=~, bJustChanged=~)
  ]
 }
 Transition {path=AnimGraph.AnimGraphNode_StateMachine_6.Head Antic.AnimStateTransitionNode_2.Transition, nodes [Result()]}
 Transition {path=AnimGraph.AnimGraphNode_StateMachine_6.Head Antic.AnimStateTransitionNode_5.Transition, nodes [Result(bCanEnterTransition=true)]}
 EventGraph {links [], nodes []}
 GetSequencerAnimSlotNames {nodes [
   GetSequencerAnimSlotNames() {is=title K2Node_FunctionEntry}
   > Return Node(ReturnValue:Array of Names=Make Array("[0]":Name=SequencerDefault) -> (Array:Array of Names))
  ]}
 OnStateEntry_0 {nodes [
   OnStateEntry_0!() -> (then, Context:"Anim Update Context Structure (by ref)", Node:"Anim Node Reference Structure (by ref)") {is=title K2Node_FunctionEntry}
   > Set bIsSpecialIdleFinish(bIsSpecialIdleFinish:bool=true) -> (Output_Get:bool)
   > Return Node!(execute)
  ]}
 OnStateExit_0 {nodes [
   OnStateExit_0!() -> (then, Context:"Anim Update Context Structure (by ref)", Node:"Anim Node Reference Structure (by ref)") {is=title K2Node_FunctionEntry}
   > Set bIsSpecialIdleFinish(bIsSpecialIdleFinish=false)
   > Return Node()
  ]}
 ToSitting_Entry {nodes [
   ToSitting_Entry!() -> (then, Context:"Anim Update Context Structure (by ref)", Node:"Anim Node Reference Structure (by ref)") {is=title K2Node_FunctionEntry}
   > Set bIsSitting(bIsSitting:bool=true) -> (Output_Get:bool)
   > Return Node()
  ]}
 ToSitting_Exit {nodes [
   ToSitting_Exit!() -> (then, Context:"Anim Update Context Structure (by ref)", Node:"Anim Node Reference Structure (by ref)") {is=title K2Node_FunctionEntry}
   > Set bIsSitting(bIsSitting=false)
   > Return Node()
  ]}
]
variables [ChooserTables {type=ProxyTable, default=/CosmeticCompanions/Assets/Biped/CrescentNail/Animation/Choosers/PRT_CrescentNail_Choosers}]
