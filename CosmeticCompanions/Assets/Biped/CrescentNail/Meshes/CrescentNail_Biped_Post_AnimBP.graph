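# Post-process AnimBP for the CrescentNail biped cosmetic companion that layers a cosmetic pose and face poses, then drives its feather bones.
# An additive pose scaled by CosmeticPoseAlpha feeds the face PoseAsset, followed by a component-space constraint chain.
# Feather drivers A–H use weighted parent constraints to jaw, neck, spine, face or clavicle bones, and each feather bone takes only translation from its driver.
# The event graph holds no logic.
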AnimBlueprint /CosmeticCompanions/Assets/Biped/CrescentNail/Meshes/CrescentNail_Biped_Post_AnimBP
target_skeleton=&/CrescentNail_Biped_Skeleton
graphs [
 AnimGraph {nodes [
   n1 = Apply Additive(
    Base=Input Pose()
    Additive=Sequence Player!(Sequence:Anim Sequence Base Object Reference=Get AdditivePoseAnim() -> (AdditivePoseAnim:Anim Sequence Object Reference)) -> (Pose)
    Alpha=Get CosmeticPoseAlpha() -> (CosmeticPoseAlpha:float)
   ) {alphaBoolBlend {blendOption=Linear}}
   n2 = "Constraint - featherDriverA_01 (PPPPPPPP)"!(
    ComponentPose=Local To Component(LocalPose=CrescentNail_face_poses_PoseAsset!(SourcePose:Pose Link Structure=@n1) -> (Pose:Pose Link Structure) {
      is=title AnimGraphNode_PoseBlendNode
      blendOption=Linear
      poseAsset=&/CrescentNail_face_poses_PoseAsset
     })
    Alpha=1
    ConstraintWeights_0=0.0144
    ConstraintWeights_1:float=0.0469
    ConstraintWeights_2:float=0.8534
    ConstraintWeights_3:float=0.0047
    ConstraintWeights_4:float=0.0199
    ConstraintWeights_5:float=0.0199
    ConstraintWeights_6:float=0.0204
    ConstraintWeights_7:float=0.0204
   ) -> (Pose) {
    alphaBoolBlend {blendOption=Linear}
    boneToModify {boneName=featherDriverA_01}
    constraintSetup [
     {targetBone {boneName=neck_02}}
     {targetBone {boneName=neck_01}}
     {targetBone {boneName=reverse_jaw}}
     {targetBone {boneName=spine_03}}
     {targetBone {boneName=cheek_tp_ot_l}}
     {targetBone {boneName=cheek_tp_ot_r}}
     {targetBone {boneName=brow_ot_r}}
     {targetBone {boneName=brow_ot_l}}
    ]
    constraintWeights [1,1,1,1,1,1,1,1]
   }
   n3 = "Constraint - featherA_01 (T)"(ComponentPose=@n2, Alpha=1, ConstraintWeights_0=1) {
    alphaBoolBlend {blendOption=Linear}
    boneToModify {boneName=featherA_01}
    constraintSetup [{transformType=Translation, targetBone {boneName=featherDriverA_01}}]
    constraintWeights [1]
   }
   n4 = "Constraint - featherDriverB_01 (PPPPPPP)"!(
    ComponentPose=@n3
    Alpha=1
    ConstraintWeights_0=0.0244
    ConstraintWeights_1:float=0.0802
    ConstraintWeights_2:float=0.8332
    ConstraintWeights_3:float=0.0052
    ConstraintWeights_4:float=0.0108
    ConstraintWeights_5:float=0.0231
    ConstraintWeights_6:float=0.0231
   ) -> (Pose) {
    alphaBoolBlend {blendOption=Linear}
    boneToModify {boneName=featherDriverB_01}
    constraintSetup [
     {targetBone {boneName=neck_02}}
     {targetBone {boneName=neck_01}}
     {targetBone {boneName=reverse_jaw}}
     {targetBone {boneName=spine_02}}
     {targetBone {boneName=spine_03}}
     {targetBone {boneName=cheek_tp_ot_l}}
     {targetBone {boneName=cheek_tp_ot_r}}
    ]
    constraintWeights [1,1,1,1,1,1,1]
   }
   n5 = "Constraint - featherB_01 (T)"(ComponentPose=@n4, Alpha=1, ConstraintWeights_0=1) {
    alphaBoolBlend {blendOption=Linear}
    boneToModify {boneName=featherB_01}
    constraintSetup [{transformType=Translation, targetBone {boneName=featherDriverB_01}}]
    constraintWeights [1]
   }
   n6 = "Constraint - featherDriverC_01 (PPPPPPP)"!(
    ComponentPose=@n5
    Alpha=1
    ConstraintWeights_0=0.0224
    ConstraintWeights_1:float=0.0743
    ConstraintWeights_2:float=0.7615
    ConstraintWeights_3:float=0.0041
    ConstraintWeights_4:float=0.0094
    ConstraintWeights_5:float=0.057
    ConstraintWeights_6:float=0.0712
   ) -> (Pose) {
    alphaBoolBlend {blendOption=Linear}
    boneToModify {boneName=featherDriverC_01}
    constraintSetup [
     {targetBone {boneName=neck_02}}
     {targetBone {boneName=neck_01}}
     {targetBone {boneName=reverse_jaw}}
     {targetBone {boneName=spine_02}}
     {targetBone {boneName=spine_03}}
     {targetBone {boneName=cheek_tp_ot_r}}
     {targetBone {boneName=brow_ot_r}}
    ]
    constraintWeights [1,1,1,1,1,1,1]
   }
   n7 = "Constraint - featherC_01 (T)"(ComponentPose=@n6, Alpha=1, ConstraintWeights_0=1) {
    alphaBoolBlend {blendOption=Linear}
    boneToModify {boneName=featherC_01}
    constraintSetup [{transformType=Translation, targetBone {boneName=featherDriverC_01}}]
    constraintWeights [1]
   }
   n8 = "Constraint - featherDriverD_01 (PPPPPPP)"!(
    ComponentPose=@n7
    Alpha=1
    ConstraintWeights_0=0.0338
    ConstraintWeights_1:float=0.1139
    ConstraintWeights_2:float=0.7095
    ConstraintWeights_3:float=0.0104
    ConstraintWeights_4:float=0.018
    ConstraintWeights_5:float=0.0525
    ConstraintWeights_6:float=0.0619
   ) -> (Pose) {
    alphaBoolBlend {blendOption=Linear}
    boneToModify {boneName=featherDriverD_01}
    constraintSetup [
     {targetBone {boneName=neck_02}}
     {targetBone {boneName=neck_01}}
     {targetBone {boneName=reverse_jaw}}
     {targetBone {boneName=spine_02}}
     {targetBone {boneName=spine_03}}
     {targetBone {boneName=cheek_tp_ot_l}}
     {targetBone {boneName=brow_ot_l}}
    ]
    constraintWeights [1,1,1,1,1,1,1]
   }
   n9 = "Constraint - featherD_01 (T)"(ComponentPose=@n8, Alpha=1, ConstraintWeights_0=1) {
    alphaBoolBlend {blendOption=Linear}
    boneToModify {boneName=featherD_01}
    constraintSetup [{transformType=Translation, targetBone {boneName=featherDriverD_01}}]
    constraintWeights [1]
   }
   n10 = "Constraint - featherDriverE_01 (PPPPPPP)"!(
    ComponentPose=@n9
    Alpha=1
    ConstraintWeights_0=0.0338
    ConstraintWeights_1:float=0.1139
    ConstraintWeights_2:float=0.7095
    ConstraintWeights_3:float=0.0104
    ConstraintWeights_4:float=0.018
    ConstraintWeights_5:float=0.0525
    ConstraintWeights_6:float=0.0619
   ) -> (Pose) {
    alphaBoolBlend {blendOption=Linear}
    boneToModify {boneName=featherDriverE_01}
    constraintSetup [
     {targetBone {boneName=neck_02}}
     {targetBone {boneName=neck_01}}
     {targetBone {boneName=reverse_jaw}}
     {targetBone {boneName=spine_02}}
     {targetBone {boneName=spine_03}}
     {targetBone {boneName=cheek_tp_ot_r}}
     {targetBone {boneName=brow_ot_r}}
    ]
    constraintWeights [1,1,1,1,1,1,1]
   }
   n11 = "Constraint - featherE_01 (T)"(ComponentPose=@n10, Alpha=1, ConstraintWeights_0=1) {
    alphaBoolBlend {blendOption=Linear}
    boneToModify {boneName=featherE_01}
    constraintSetup [{transformType=Translation, targetBone {boneName=featherDriverE_01}}]
    constraintWeights [1]
   }
   n12 = "Constraint - featherDriverF_01 (PPPPPPPP)"!(
    ComponentPose=@n11
    Alpha=1
    ConstraintWeights_0=0.0583
    ConstraintWeights_1:float=0.2016
    ConstraintWeights_2:float=0.0129
    ConstraintWeights_3:float=0.6452
    ConstraintWeights_4:float=0.0321
    ConstraintWeights_5:float=0.0461
    ConstraintWeights_6:float=0.0019
    ConstraintWeights_7:float=0.0019
   ) -> (Pose) {
    alphaBoolBlend {blendOption=Linear}
    boneToModify {boneName=featherDriverF_01}
    constraintSetup [
     {targetBone {boneName=neck_02}}
     {targetBone {boneName=neck_01}}
     {targetBone {boneName=spine_01}}
     {targetBone {boneName=reverse_jaw}}
     {targetBone {boneName=spine_02}}
     {targetBone {boneName=spine_03}}
     {targetBone {boneName=lip_corner_l}}
     {targetBone {boneName=lip_corner_r}}
    ]
    constraintWeights [1,1,1,1,1,1,1,1]
   }
   n13 = "Constraint - featherF_01 (T)"(ComponentPose=@n12, Alpha=1, ConstraintWeights_0=1) {
    alphaBoolBlend {blendOption=Linear}
    boneToModify {boneName=featherF_01}
    constraintSetup [{transformType=Translation, targetBone {boneName=featherDriverF_01}}]
    constraintWeights [1]
   }
   n14 = "Constraint - featherDriverG_01 (PPPPPPPP)"!(
    ComponentPose=@n13
    Alpha=1
    ConstraintWeights_0=0.0698
    ConstraintWeights_1:float=0.2678
    ConstraintWeights_2:float=0.0359
    ConstraintWeights_3:float=0.0043
    ConstraintWeights_4:float=0.4425
    ConstraintWeights_5:float=0.0043
    ConstraintWeights_6:float=0.0729
    ConstraintWeights_7:float=0.1025
   ) -> (Pose) {
    alphaBoolBlend {blendOption=Linear}
    boneToModify {boneName=featherDriverG_01}
    constraintSetup [
     {targetBone {boneName=neck_02}}
     {targetBone {boneName=neck_01}}
     {targetBone {boneName=spine_01}}
     {targetBone {boneName=clavicle_l}}
     {targetBone {boneName=reverse_jaw}}
     {targetBone {boneName=clavicle_r}}
     {targetBone {boneName=spine_02}}
     {targetBone {boneName=spine_03}}
    ]
    constraintWeights [1,1,1,1,1,1,1,1]
   }
   n15 = "Constraint - featherG_01 (T)"(ComponentPose=@n14, Alpha=1, ConstraintWeights_0=1) {
    alphaBoolBlend {blendOption=Linear}
    boneToModify {boneName=featherG_01}
    constraintSetup [{transformType=Translation, targetBone {boneName=featherDriverG_01}}]
    constraintWeights [1]
   }
   n16 = "Constraint - featherDriverH_01 (PPPPPPPP)"!(
    ComponentPose=@n15
    Alpha=1
    ConstraintWeights_0=0.0643
    ConstraintWeights_1:float=0.2948
    ConstraintWeights_2:float=0.0752
    ConstraintWeights_3:float=0.0005
    ConstraintWeights_4:float=0.2163
    ConstraintWeights_5:float=0.0233
    ConstraintWeights_6:float=0.1376
    ConstraintWeights_7:float=0.1879
   ) -> (Pose) {
    alphaBoolBlend {blendOption=Linear}
    boneToModify {boneName=featherDriverH_01}
    constraintSetup [
     {targetBone {boneName=neck_02}}
     {targetBone {boneName=neck_01}}
     {targetBone {boneName=spine_01}}
     {targetBone {boneName=clavicle_l}}
     {targetBone {boneName=reverse_jaw}}
     {targetBone {boneName=clavicle_r}}
     {targetBone {boneName=spine_02}}
     {targetBone {boneName=spine_03}}
    ]
    constraintWeights [1,1,1,1,1,1,1,1]
   }
   n17 = "Constraint - featherH_01 (T)"(ComponentPose=@n16, Alpha=1, ConstraintWeights_0=1) {
    alphaBoolBlend {blendOption=Linear}
    boneToModify {boneName=featherH_01}
    constraintSetup [{transformType=Translation, targetBone {boneName=featherDriverH_01}}]
    constraintWeights [1]
   }
   Output Pose(Result=Component To Local(ComponentPose=@n17))
  ]}
 EventGraph {nodes [
   TryGetPawnOwner()
   Event BlueprintUpdateAnimation()
  ]}
]
variables []
